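# Editor utility task that runs asset error scans, executing each asset's list of Base Error Task tests as child tasks.
# It precounts steps, skips tests listed in assetErrorsToSkip, using Python to resolve test class package names, and registers the rest in one loop.
# Completion callbacks broadcast progress through onTestStepComplete and group failures per asset in finalErrors.
# A step counter, advanced by finished and skipped tests, signals when the whole task finishes.
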EditorUtilityBlueprint /Game/Blueprints/Blutilities/ForTechArt/AssetErrorScanner/Tasks/AssetScanForError_Task
graphs [
 EventGraph {nodes [
   Event BeginExecution()
   > Set wasCanceled(wasCanceled:bool=false) -> (Output_Get:bool)
   > getTotalTestSteps()
   > n1 = Run Tests!(Exec:Exec) -> (" ":Exec, AllTasksComplete:Exec) {is=title K2Node_MacroInstance}
   FinishExecutingTask(execute=@n1.AllTasksComplete)
  ]}
 HandleEachTestComplete {nodes [
   n1 = HandleEachTestComplete!() -> (then, Task:Object Reference) {is=title K2Node_FunctionEntry}
   > n2 = Cast To Base_Error_Task!(execute, Object=@n1) -> (then, CastFailed, AsBase Error Task:Base Error Task Object Reference) {is=title K2Node_DynamicCast}
   > n3 = Sequence()
   n4 = "To Text (Integer)"(Value=Get totalSteps() -> (totalSteps:int=0), MinimumIntegralDigits=2)
   n5 = "To Text (Integer)"(Value=Get currentStep() -> (currentStep:int=0), MinimumIntegralDigits=2)
   n6 = Get currentStep() -> (currentStep=0)
   n7 = Get stepNames() -> (stepNames:Array of Strings)
   > Call On Test Step Complete!(
    execute:Exec
    self:Asset Scan For Error Task Object Reference
    EventName:String=Append!(
     A="To String (Text)"(InText=@n5)
     B="/ "
     C:String="To String (Text)"(InText=@n4)
     D:String=" "
     E:String="Get (a copy)"(Array=@n7, Dimension 1=@n6)
    ) -> (ReturnValue)
    stepIndex:int=Get currentStep() -> (currentStep=0)
    stepTotal:int=Get totalSteps() -> (totalSteps=0)
   ) -> (then:Exec) {is=title K2Node_CallDelegate}
   n8 = getFinalErrorData!(execute=@n3.then_1, self:Base Error Task Object Reference=@n2) -> (then, isError:bool, errorData:AVT Error Struct Structure)
   > Branch(Condition=@n8)
   > Set l_newError(l_newError:AVT Error Struct Structure=@n8.errorData) -> (Output_Get:AVT Error Struct Structure)
   > Set l_assetName(l_assetName:Name=Break Asset Data(AssetData=Get assetData!(self:Base Error Task Object Reference=@n2) -> (assetData:Asset Data Structure))) -> (Output_Get:Name=None)
   n9 = Find(
    TargetMap:Map of Names to AES Test Result Struct Structures=Get finalErrors() -> (finalErrors:Map of Names to AES Test Result Struct Structures)
    Key:"Name (by ref)"=Get l_assetName() -> (l_assetName:Name=None)
   ) -> (Value:AES Test Result Struct Structure)
   > n10 = Branch(Condition=@n9.ReturnValue)
   > Set l_struct(l_struct:AES Test Result Struct Structure=@n9) -> (Output_Get:AES Test Result Struct Structure)
   > n11 = Add(
    TargetArray:Array of AVT Error Struct Structures=Break AES Test Result Struct!(AES_TestResult_Struct:"AES Test Result Struct Structure (by ref)"=Get l_struct() -> (l_struct:AES Test Result Struct Structure)) -> (assetPackageName_15_#1:Name, errors_17_#2:Array of AVT Error Struct Structures) {is=title K2Node_BreakStruct}.errors_17_#2
    NewItem:"AVT Error Struct Structure (by ref)"=Get l_newError() -> (l_newError:AVT Error Struct Structure)
   ) {is=title K2Node_CallArrayFunction}
   Add!(
    execute=[@n11.then, @n13.then]
    TargetMap:Map of Names to AES Test Result Struct Structures=Get finalErrors()
    Key:"Name (by ref)"=Get l_assetName() -> (l_assetName=None)
    Value:"AES Test Result Struct Structure (by ref)"=Get l_struct()
   ) -> (then) {is=title K2Node_CallFunction}
   n12 = Get l_newError()
   n13 = Set l_struct(
    execute=@n10.else
    l_struct=Make AES Test Result Struct!(
     assetPackageName_15_#1:Name=Get l_assetName() -> (l_assetName=None)
     errors_17_#2:Array of AVT Error Struct Structures=Make Array("[0]":AVT Error Struct Structure=@n12) -> (Array:Array of AVT Error Struct Structures)
    ) -> (AES_TestResult_Struct:AES Test Result Struct Structure) {is=title K2Node_MakeStruct}
   )
  ]}
 RunTests {nodes [
   Inputs!() -> (Exec:Exec)
   > Clear!(execute, TargetMap:Map of Names to AES Test Result Struct Structures=Get finalErrors()) -> (then) {is=title K2Node_CallFunction}
   > n1 = Sequence()
   > n2 = For Each Loop(Array:Array of AES Tests Struct Structures=Get assetTests() -> (assetTests:Array of AES Tests Struct Structures)) -> (Array Element:AES Tests Struct Structure)
   > Set Notification Text!(execute, self, AES_Tests_Struct:"AES Tests Struct Structure (by ref)"=@n2) -> (then) {is=title K2Node_CallFunction}
   n3 = Break AES Tests Struct!(AES_Tests_Struct:"AES Tests Struct Structure (by ref)"=@n2) -> (
    assetData_9_#1:Asset Data Structure
    asset_8_#3:Object Reference
    tests_10_#2:Array of Base Error Task Class References
   ) {is=title K2Node_BreakStruct}
   > n4 = For Each Loop(Array:Array of Base Error Task Class References=@n3.tests_10_#2) -> (Array Element:Base Error Task Class Reference)
   > n5 = Construct Object from Class!(execute, Class=@n4, self, assetData:Asset Data Structure=@n3, asset:Object Reference=@n3.asset_8_#3) -> (then, ReturnValue:Base Error Task Object Reference)
   > n6 = getTaskPackageNameFromClass!(execute, self, in_object:Base Error Task Class Reference=@n4) -> (then, PackageName:Name=None)
   > n7 = shouldRunTest!(
    execute
    self
    assetPackageName:"Name (by ref)"=Break Asset Data!(AssetData=@n3) -> (PackageName, AssetClassPath)
    taskPackageName:"Name (by ref)"=@n6
   ) -> (then, ReturnValue:bool)
   > n8 = Branch(Condition=@n7)
   > n9 = RegisterAndExecuteTask!(execute:Exec, Task:Editor Utility Task Object Reference=@n5, OptionalParentTask:Editor Utility Task Object Reference=Self-Reference()) -> (then:Exec, OnFinished:Exec, Task:Editor Utility Task Object Reference) {is=title K2Node_AsyncAction}
   n10 = Break Asset Data!(AssetData=@n3) -> (AssetName:Name, AssetClassPath)
   > Call On Test Step Complete(
    EventName=Append!(A="To String (Name)"(InName=@n10), B=" Loading..") -> (ReturnValue)
    stepIndex=Get currentStep() -> (currentStep=0)
    stepTotal=Get totalSteps() -> (totalSteps=0)
   )
   n11 = Branch(execute=@n9.OnFinished, Condition=Get wasCanceled() -> (wasCanceled:bool))
   n12 = Increment Int(" "=[@n11.then, @n8.else], Value=Get currentStep() -> (currentStep=0))
   n13 = Get currentStep() -> (currentStep=0)
   n14 = Branch(execute=[@n15.then, @n12."  "], Condition="Equal (Integer)"(A=@n13, B=Get totalSteps() -> (totalSteps=0)))
   Outputs!(" ":Exec=@n1.then_1, AllTasksComplete:Exec=@n14.then)
   Increment Int(" "=@n11.else, Value=@n13)
   > n15 = HandleEachTestComplete!(execute, self, Task:Object Reference=@n9) -> (then) {is=title K2Node_CallFunction}
  ]}
 Set Notification Text {nodes [
   n1 = Set Notification Text!() -> (then, AES_Tests_Struct:"AES Tests Struct Structure (by ref)") {is=title K2Node_FunctionEntry}
   n2 = Break AES Tests Struct(AES_Tests_Struct=@n1)
   > n3 = Is Valid(InputObject=@n2.asset_8_#3)
   n4 = Set l_assetName(
    execute=@n3.Is Not Valid
    l_assetName:String="To String (Name)"(InName=Break Asset Data!(AssetData=@n2) -> (PackageName:Name, PackagePath:Name, AssetName, AssetClassPath).AssetName)
   ) -> (Output_Get:String)
   n5 = Get Class Defaults!(Class:Object Class Reference="Get (a copy)"(Array:Array of Base Error Task Class References=@n2.tests_10_#2) -> (Output:Base Error Task Class Reference)) -> (assetData:Asset Data Structure, asset:Object Reference, friendlyName:String, Is Error:bool=~, finalErrorData:AVT Error Struct Structure) {is=title K2Node_GetClassDefaults}
   n6 = Get l_assetName() -> (l_assetName:String)
   SetTaskNotificationText!(
    execute=[@n4.then, @n7.then]
    self:Editor Utility Task Object Reference
    Text:"Text (by ref)"="To Text (String)"(InString=Append!(A=@n6, B="Running ", C:String=@n5.friendlyName) -> (ReturnValue))
   ) -> (then)
   n7 = Set l_assetName(execute=@n3.Is Valid, l_assetName=GetDisplayName(Object=@n2.asset_8_#3))
  ]}
 cancelScan {nodes [cancelScan() {is=title K2Node_FunctionEntry}]}
 getTaskPackageNameFromClass {nodes [
   n1 = getTaskPackageNameFromClass!() -> (then, in_object:Base Error Task Class Reference) {is=title K2Node_FunctionEntry}
   > n2 = Execute Python Script!(
    execute
    PythonScript="import unreal\r\nfinal_package_name = str(in_object).split(\"Object '\")[1]\r\nfinal_package_name = final_package_name.split(\".\")[0]\r\nfinal_asset = unreal.load_asset(final_package_name)\r\n##print(\"final_asset \" + str(final_asset))\r\nasset_data = unreal.AssetRegistryHelpers.create_asset_data(final_asset)\r\n##print(\"asset_data \" + str(asset_data))"
    in_object:Base Error Task Class Reference=@n1
   ) -> (then, ReturnValue, asset_data:"Asset Data Structure (by ref)") {inputs [in_object], outputs [asset_data]}
   > Return Node!(execute, PackageName:Name=Break Asset Data!(AssetData=@n2.asset_data) -> (PackageName, AssetClassPath))
  ]}
 getTotalTestSteps {nodes [
   getTotalTestSteps() {is=title K2Node_FunctionEntry}
   > n1 = For Each Loop(Array:Array of AES Tests Struct Structures=Get assetTests()) -> (Array Element:AES Tests Struct Structure)
   n2 = Break AES Tests Struct(AES_Tests_Struct=@n1)
   > Set totalSteps(totalSteps:int=int + int(
     A=Get totalSteps() -> (totalSteps=0)
     B=Length(TargetArray:Array of Base Error Task Class References=@n2.tests_10_#2)
    )) -> (Output_Get:int=0)
   > n3 = For Each Loop(Array:Array of Base Error Task Class References=@n2.tests_10_#2) -> (Array Element:Base Error Task Class Reference)
   n4 = Break Asset Data!(AssetData=@n2) -> (PackageName, PackagePath:Name, AssetName:Name, AssetClassPath)
   > Add(
    TargetArray:Array of Strings=Get stepNames()
    NewItem:"String (by ref)"=Append(A="To String (Name)"(InName=@n4.AssetName), B=" - ", C=Get Class Defaults(Class=@n3) -> (Is Error=~).friendlyName)
   ) {is=title K2Node_CallArrayFunction}
  ]}
 onTestStepComplete {nodes [onTestStepComplete!() -> (then, EventName:String, stepIndex:int, stepTotal:int) {is=title K2Node_FunctionEntry}]}
 shouldRunTest {nodes [
   n1 = shouldRunTest!() -> (then, assetPackageName:"Name (by ref)", taskPackageName:"Name (by ref)") {is=title K2Node_FunctionEntry}
   n2 = Find(
    TargetMap:Map of Names to AES Skip Test Struct Structures=Get assetErrorsToSkip() -> (assetErrorsToSkip:Map of Names to AES Skip Test Struct Structures)
    Key=@n1
   ) -> (Value:AES Skip Test Struct Structure)
   > n3 = Branch(Condition=@n2.ReturnValue)
   > n4 = Branch(Condition=Contains Item(
     TargetArray=Break AES Skip Test Struct!(AES_SkipTest_Struct:"AES Skip Test Struct Structure (by ref)"=@n2) -> (assetPackageName_14_#1:Name, skippedTests_15_#2:Array of Names) {is=title K2Node_BreakStruct}.skippedTests_15_#2
     ItemToFind=@n1.taskPackageName
    ))
   > Return Node!(execute, ReturnValue:bool=false)
   Return Node(execute=@n4.else, ReturnValue=true)
   Return Node(execute=@n3.else, ReturnValue=true)
  ]}
]
variables [
 assetErrorsToSkip {type="map<AES_SkipTest_Struct>", default {}}
 assetTests {type="array<AES_Tests_Struct>", default []}
 currentStep {type=integer}
 finalErrors {type="map<AES_TestResult_Struct>", default {}}
 onTestStepComplete {default=()}
 stepNames {type="array<string>", default []}
 totalSteps {type=integer}
 wasCanceled {type=boolean}
]
