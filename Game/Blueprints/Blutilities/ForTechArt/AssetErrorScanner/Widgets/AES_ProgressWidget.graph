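# AES_ProgressWidget is a UMG progress overlay, filed under the tech-art AssetErrorScanner Blutility tools, that shows a status line and a progress bar.
# A dark, fixed-height panel inside a translucent border stacks a text block above a blue bar.
# The setProgress function takes a string and a percent, writing them straight to the text and bar.
# It holds no variables, and its event graph contains only empty default events.
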WidgetBlueprint /Game/Blueprints/Blutilities/ForTechArt/AssetErrorScanner/Widgets/AES_ProgressWidget
graphs [
 EventGraph {nodes [
   Event PreConstruct()
   Event Construct()
   Event Tick()
  ]}
 setProgress {nodes [
   n1 = setProgress!() -> (then, progressString:String, percent:double=~) {is=title K2Node_FunctionEntry}
   > "SetText (Text)"(self=Get CommonTextBlock_104() -> (CommonTextBlock_104:Common Text Object Reference), InText="To Text (String)"(InString=@n1))
   > SetPercent!(
    execute
    self:Progress Bar Object Reference=Get ProgressBar_88() -> (ProgressBar_88:Progress Bar Object Reference)
    InPercent:float=@n1.percent
   ) -> (then)
  ]}
]
variables []
widgets {
 info {inheritedWidgetCount=0, parentClass=&UserWidget, rootWidgetClass=&Border, widgetCount=6}
 widgets [
  Border_0 {
   widgetClassPath=&Border
   widget=Border_0
   properties {verticalAlignment=VAlign_Center, brushColor=<rgba 0 0 0 0.359>, padding=<ltrb 50 50 50 50>}
  }
  SizeBox_0 {
   widgetClassPath=&SizeBox
   parent=Border_0
   widget=SizeBox_0
   properties {bOverride_HeightOverride=true, heightOverride=100, widthOverride=1048.8608}
   slot {class=BorderSlot, properties {verticalAlignment=VAlign_Center, padding=<ltrb 50 50 50 50>}}
  }
  Border_223 {
   widgetClassPath=&Border
   parent=SizeBox_0
   widget=Border_223
   properties {brushColor=<rgba 0.0156 0.0156 0.0156 1>}
   slot {class=SizeBoxSlot}
  }
  VerticalBox_68 {widgetClassPath=&VerticalBox, parent=Border_223, widget=VerticalBox_68, slot {class=BorderSlot}}
  CommonTextBlock_104 {
   widgetClassPath=&CommonTextBlock
   bIsVariable=true
   parent=VerticalBox_68
   widget=CommonTextBlock_104
   properties {
    text=Text Block
    font {
     bForceMonospaced=false
     bMaterialIsStencil=false
     fontMaterial=null
     fontObject=&NotoSans
     letterSpacing=0
     monospacedWidth=1
     size=18
     skewAmount=0
     typefaceFontName=NotoSans-Regular
     outlineSettings {
      bApplyOutlineToDropShadows=false
      bMiteredCorners=false
      bSeparateFillAlpha=false
      outlineMaterial=null
      outlineSize=0
      outlineColor=<rgba 0 0 0 1>
     }
    }
    shadowOffset=<0 0>
   }
   slot {class=VerticalBoxSlot, properties {horizontalAlignment=HAlign_Left, padding=<ltrb 20 5 5 5>}}
  }
  ProgressBar_88 {
   widgetClassPath=/Script/UMG.ProgressBar
   bIsVariable=true
   parent=VerticalBox_68
   widget=ProgressBar_88
   properties {fillColorAndOpacity=<rgba 0 0.5 1 1>}
   slot {class=VerticalBoxSlot, properties {verticalAlignment=VAlign_Center, padding=<ltrb 20 20 20 20>, size {sizeRule=Fill, value=1}}}
  }
 ]
}
